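# Animation Blueprint for the female medium KeyTracker cosmetic body, which follows its attached parent mesh and adds secondary motion to the laces.
# Copy Pose From Mesh feeds two chained AnimDynamics nodes that simulate the left and right lace bones in root-relative space with angular springs and limits.
# The result then passes through the KeyTracker Control Rig before output.
# The event graph holds only an unconnected update event, with no variables or logic.
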AnimBlueprint /BRCosmetics/Characters/Player/Female/Medium/Bodies/F_MED_KeyTracker/Meshes/F_MED_KeyTracker_AnimBP
target_skeleton=&/F_MED_KeyTracker_Skeleton
graphs [
 AnimGraph {nodes [
   n1 = Anim Dynamics - Bone: dyn_lace_l(ComponentPose=Local To Component(LocalPose=Copy Pose From Mesh() {bUseAttachedParent=true}), Alpha=1, ExternalForce="0,0,0") {
    angularDampingOverride=1
    angularSpringConstant=100
    bOverrideAngularDamping=true
    simulationSpace=RootRelative
    alphaBoolBlend {blendOption=Linear}
    boundBone {boneName=dyn_lace_l}
    componentAppliedLinearAccClamp=<1000 1000 1000>
    componentLinearAccScale=<0.3 0.3 0.3>
    componentLinearVelScale=<0.3 0.3 0.3>
    gravityOverride {z=-980}
    physicsBodyDefinitions [{
      boundBone {boneName=dyn_lace_l}
      boxExtents=<2 2 2>
      constraintSetup {angularLimitsMax=<2 30 30>, angularLimitsMin=<-2 -30 -10>, angularTarget=<1 0 0>}
      localJointOffset=<0.1 0 0>
     }]
    relativeSpaceBone {boneName=calf_twist_02_r}
   }
   n2 = Anim Dynamics - Bone: dyn_lace_r(ComponentPose=@n1, Alpha=1, ExternalForce="0,0,0") {
    angularDampingOverride=1
    angularSpringConstant=100
    bOverrideAngularDamping=true
    simulationSpace=RootRelative
    alphaBoolBlend {blendOption=Linear}
    boundBone {boneName=dyn_lace_r}
    componentAppliedLinearAccClamp=<1000 1000 1000>
    componentLinearAccScale=<0.3 0.3 0.3>
    componentLinearVelScale=<0.3 0.3 0.3>
    gravityOverride {z=-980}
    physicsBodyDefinitions [{
      boundBone {boneName=dyn_lace_r}
      boxExtents=<2 2 2>
      constraintSetup {angularLimitsMax=<2 30 30>, angularLimitsMin=<-2 -30 -10>, angularTarget=<1 0 0>}
      localJointOffset=<0.1 0 0>
     }]
    relativeSpaceBone {boneName=calf_twist_02_r}
   }
   n3 = Control Rig(Source=Component To Local(ComponentPose=@n2)) {
    alphaBoolBlend {blendOption=Linear}
    controlRigAssetReference {blueprintRigClass=&/F_MED_KeyTracker_ControlRig.F_MED_KeyTracker_ControlRig_C}
   }
   Output Pose(Result=@n3) {layerGroup=None}
  ]}
 EventGraph {nodes [Event BlueprintUpdateAnimation()]}
]
variables []
