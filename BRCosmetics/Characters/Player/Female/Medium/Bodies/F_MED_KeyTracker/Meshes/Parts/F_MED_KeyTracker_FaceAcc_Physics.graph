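# Physics asset for the KeyTracker female character's face accessory part, defining bodies for its dynamic bangs, tails and branch bones.
# Two six-segment tail chains, the bangs and two base spheres hang off the head with tight angular limits, and each base fans out to seven individually constrained branch capsules.
# Head, neck, upper spine, both clavicles and only the right upper arm add body shapes, with free angular constraints along the right arm chain.
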PhysicsAsset /BRCosmetics/Characters/Player/Female/Medium/Bodies/F_MED_KeyTracker/Meshes/Parts/F_MED_KeyTracker_FaceAcc_Physics
bodies {
 clavicle_l [{
   shapeType=Capsule
   length=8.946
   radius=4.7513
   shapeName=clavicle_l_capsule
   center=<4.7745 0.0002 -0.0527>
   rotation=<rot -83.9934 179.8493 -179.8474>
  }]
 clavicle_r [{
   shapeType=Capsule
   length=8.946
   radius=4.7513
   shapeName=clavicle_l_capsule
   center=<-4.7745 -0.0002 0.0527>
   rotation=<rot -83.9934 179.8493 0.1526>
  }]
 dyn_bangs [{shapeType=Capsule, length=5.3814, radius=1.4697, shapeName=dyn_bangs_capsule, center=<4.6866 0 1.497>, rotation=<rot -83.2851 0 0>}]
 dyn_base_l [{shapeType=Sphere, radius=3.3401, shapeName=dyn_base_l_sphere}]
 dyn_base_r [{shapeType=Sphere, radius=3.3401, shapeName=dyn_base_l_sphere}]
 dyn_branch_l_1 [{
   shapeType=Capsule
   length=3
   radius=1
   shapeName=dyn_branch_l_1_capsule
   center=<8.4654 0.7545 0.3604>
   rotation=<rot 0.1787 -84.9145 -24.5753>
  }]
 dyn_branch_l_2 [{
   shapeType=Capsule
   length=3
   radius=1
   shapeName=dyn_branch_l_2_capsule
   center=<4.4012 -0.1903 0.3957>
   rotation=<rot 3.082 -92.7527 -62.3621>
  }]
 dyn_branch_l_3 [{
   shapeType=Capsule
   length=3
   radius=1
   shapeName=dyn_branch_l_3_capsule
   center=<4.1636 -0.2928 0.8798>
   rotation=<rot 10.6118 -96.2866 -99.1584>
  }]
 dyn_branch_l_4 [{
   shapeType=Capsule
   length=3
   radius=1
   shapeName=dyn_branch_l_4_capsule
   center=<2.8074 0.403 2.3421>
   rotation=<rot 15.077 -140.8608 -131.7192>
  }]
 dyn_branch_l_5 [{
   shapeType=Capsule
   length=3
   radius=1
   shapeName=dyn_branch_l_5_capsule
   center=<2.8354 1.2373 -1.1249>
   rotation=<rot 13.7492 -60.103 11.2775>
  }]
 dyn_branch_l_6 [{
   shapeType=Capsule
   length=3
   radius=1
   shapeName=dyn_branch_l_6_capsule
   center=<3.8649 0.7386 0.2064>
   rotation=<rot -2.5278 -79.0489 -41.9242>
  }]
 dyn_branch_l_7 [{
   shapeType=Capsule
   length=3
   radius=1
   shapeName=dyn_branch_l_7_capsule
   center=<2.5556 0.0936 3.1152>
   rotation=<rot 0.0001 92.0975 -53.5687>
  }]
 dyn_branch_r_1 [{
   shapeType=Capsule
   length=3
   radius=1
   shapeName=dyn_branch_r_1_capsule
   center=<6.0519 -1.7143 -0.8629>
   rotation=<rot 0.0037 -105.815 -29.3706>
  }]
 dyn_branch_r_2 [{
   shapeType=Capsule
   length=3
   radius=1
   shapeName=dyn_branch_r_2_capsule
   center=<6.4055 -1.0289 -1.1152>
   rotation=<rot -8.4641 82.34 -57.5192>
  }]
 dyn_branch_r_3 [{
   shapeType=Capsule
   length=3
   radius=1
   shapeName=dyn_branch_r_3_capsule
   center=<4.1812 0.4664 0.3223>
   rotation=<rot 1.7484 96.4991 -85.6171>
  }]
 dyn_branch_r_4 [{
   shapeType=Capsule
   length=3
   radius=1
   shapeName=dyn_branch_r_4_capsule
   center=<2.3063 0.3659 0.203>
   rotation=<rot 0.0002 -80.9855 -94.9676>
  }]
 dyn_branch_r_5 [{shapeType=Capsule, length=3, radius=1, shapeName=dyn_branch_r_5_capsule, center=<4.4786 0.4307 0.3894>, rotation=<rot 0 -84.5068 -28.4365>}]
 dyn_branch_r_6 [{
   shapeType=Capsule
   length=3
   radius=1
   shapeName=dyn_branch_r_6_capsule
   center=<5.6594 -0.8989 0.465>
   rotation=<rot 3.6974 28.1909 -87.1961>
  }]
 dyn_branch_r_7 [{
   shapeType=Capsule
   length=3
   radius=1
   shapeName=dyn_branch_r_7_capsule
   center=<1.8864 0.8305 1.6884>
   rotation=<rot 0.0001 -66.2381 53.9933>
  }]
 dyn_tale_l_0 [{
   shapeType=Capsule
   length=2.4607
   radius=0.7808
   shapeName=dyn_tale_l_0_capsule
   center=<1.2962 0.071 -0.1347>
   rotation=<rot 0 82.7212 -80.0311>
  }]
 dyn_tale_l_1 [{
   shapeType=Capsule
   length=2.3369
   radius=0.5
   shapeName=dyn_tale_l_1_capsule
   center=<1.4321 -0.0078 -0.0418>
   rotation=<rot 0 91.1626 -85.3143>
  }]
 dyn_tale_l_2 [{
   shapeType=Capsule
   length=2.5314
   radius=0.5239
   shapeName=dyn_tale_l_2_capsule
   center=<1.9452 -0.0208 0.0127>
   rotation=<rot 0 89.6633 -89.4355>
  }]
 dyn_tale_l_3 [{
   shapeType=Capsule
   length=1.9631
   radius=0.5
   shapeName=dyn_tale_l_3_capsule
   center=<1.3286 -0.0015 0.0679>
   rotation=<rot 0 88.0737 -88.3387>
  }]
 dyn_tale_l_4 [{
   shapeType=Capsule
   length=1.3631
   radius=0.5098
   shapeName=dyn_tale_l_4_capsule
   center=<1.1063 -0.0071 0.0477>
   rotation=<rot 0 88.4466 -88.4584>
  }]
 dyn_tale_l_5 [{
   shapeType=Capsule
   length=1.9563
   radius=0.5612
   shapeName=dyn_tale_l_5_capsule
   center=<1.0728 0.0582 0.0737>
   rotation=<rot 0 95.0407 -88.5709>
  }]
 dyn_tale_r_0 [{
   shapeType=Capsule
   length=2.4607
   radius=0.7808
   shapeName=dyn_tale_r_0_capsule
   center=<-1.2898 -0.0708 0.1045>
   rotation=<rot 0 82.7211 -80.0311>
  }]
 dyn_tale_r_1 [{
   shapeType=Capsule
   length=2.0782
   radius=0.5
   shapeName=dyn_tale_l_1_capsule
   center=<-1.3525 0.0094 0.0484>
   rotation=<rot 0 91.1626 94.6857>
  }]
 dyn_tale_r_2 [{
   shapeType=Capsule
   length=2.333
   radius=0.5239
   shapeName=dyn_tale_l_2_capsule
   center=<-2.1624 0.0221 -0.0148>
   rotation=<rot 0 89.6633 90.5645>
  }]
 dyn_tale_r_3 [{
   shapeType=Capsule
   length=1.9631
   radius=0.5
   shapeName=dyn_tale_l_3_capsule
   center=<-1.3286 0.0015 -0.0679>
   rotation=<rot 0 88.0737 91.6613>
  }]
 dyn_tale_r_4 [{
   shapeType=Capsule
   length=1.3631
   radius=0.5098
   shapeName=dyn_tale_l_4_capsule
   center=<-1.1063 0.0071 -0.0477>
   rotation=<rot 0 88.4466 91.5416>
  }]
 dyn_tale_r_5 [{
   shapeType=Capsule
   length=1.9563
   radius=0.5612
   shapeName=dyn_tale_l_5_capsule
   center=<-1.0728 -0.0582 -0.0737>
   rotation=<rot 0 95.0407 91.4291>
  }]
 head [
  {shapeType=Capsule, length=8.6812, radius=7.1763, center=<4.3013 1.8111 0.1443>, rotation=<rot 0 87.0369 -87.8494>}
  {
   shapeType=Capsule
   length=11.0664
   radius=4.2384
   shapeName=head_capsule
   center=<-4.7185 8.5797 -9.8486>
   rotation=<rot 2.6756 6.28 60.6276>
  }
 ]
 neck_01 [{
   shapeType=Capsule
   length=8.4773
   radius=4.0837
   shapeName=neck_01_capsule
   center=<2.6464 0.9226 0.0534>
   rotation=<rot 88.8437 179.9999 -180>
  }]
 spine_05 [{
   shapeType=Capsule
   length=13.3488
   radius=16.6512
   shapeName=spine_05_capsule
   center=<-8.4526 -4.0229 -0.0325>
   rotation=<rot 89.7118 179.9969 -146.1614>
  }]
 upperarm_r [{
   shapeType=Capsule
   length=17.8476
   radius=3.9557
   shapeName=upperarm_l_capsule
   center=<-9.1664 0.7475 0.4995>
   rotation=<rot -88.9721 -176.5991 -4.6291>
  }]
}
constraints [
 {bone1Name=dyn_tale_l_2, bone2Name=dyn_tale_l_1, swing1LimitDegrees=20, swing2LimitDegrees=15, twistLimitDegrees=3}
 {bone1Name=dyn_tale_l_3, bone2Name=dyn_tale_l_2, swing1LimitDegrees=15, swing2LimitDegrees=15, twistLimitDegrees=3}
 {bone1Name=dyn_tale_l_4, bone2Name=dyn_tale_l_3, swing1LimitDegrees=20, swing2LimitDegrees=20, twistLimitDegrees=20}
 {bone1Name=dyn_tale_l_5, bone2Name=dyn_tale_l_4, swing1LimitDegrees=20, swing2LimitDegrees=20, twistLimitDegrees=20}
 {bone1Name=dyn_tale_r_1, bone2Name=dyn_tale_r_0, swing1LimitDegrees=20, swing2LimitDegrees=5, twistLimitDegrees=2}
 {bone1Name=dyn_tale_r_2, bone2Name=dyn_tale_r_1, swing1LimitDegrees=20, swing2LimitDegrees=15, twistLimitDegrees=3}
 {bone1Name=dyn_tale_r_3, bone2Name=dyn_tale_r_2, swing1LimitDegrees=15, swing2LimitDegrees=15, twistLimitDegrees=3}
 {bone1Name=dyn_tale_r_4, bone2Name=dyn_tale_r_3, swing1LimitDegrees=20, swing2LimitDegrees=20, twistLimitDegrees=20}
 {bone1Name=dyn_tale_r_5, bone2Name=dyn_tale_r_4, swing1LimitDegrees=20, swing2LimitDegrees=20, twistLimitDegrees=20}
 {bone1Name=dyn_base_r, bone2Name=head, swing1LimitDegrees=2, swing2LimitDegrees=2, twistLimitDegrees=2}
 {bone1Name=dyn_base_l, bone2Name=head, swing1LimitDegrees=2, swing2LimitDegrees=2, twistLimitDegrees=2}
 {bone1Name=dyn_branch_l_1, bone2Name=dyn_base_l, swing1LimitDegrees=20, swing2LimitDegrees=20, twistLimitDegrees=10}
 {bone1Name=dyn_branch_l_2, bone2Name=dyn_base_l, swing1LimitDegrees=20, swing2LimitDegrees=20, twistLimitDegrees=10}
 {bone1Name=dyn_branch_l_3, bone2Name=dyn_base_l, swing1LimitDegrees=20, swing2LimitDegrees=20, twistLimitDegrees=10}
 {bone1Name=dyn_branch_l_4, bone2Name=dyn_base_l, swing1LimitDegrees=20, swing2LimitDegrees=20, twistLimitDegrees=10}
 {bone1Name=dyn_branch_l_5, bone2Name=dyn_base_l, swing1LimitDegrees=20, swing2LimitDegrees=20, twistLimitDegrees=10}
 {bone1Name=dyn_branch_l_6, bone2Name=dyn_base_l, swing1LimitDegrees=20, swing2LimitDegrees=20, twistLimitDegrees=10}
 {bone1Name=dyn_branch_l_7, bone2Name=dyn_base_l, swing1LimitDegrees=20, swing2LimitDegrees=20, twistLimitDegrees=10}
 {bone1Name=dyn_branch_r_1, bone2Name=dyn_base_r, swing1LimitDegrees=20, swing2LimitDegrees=20, twistLimitDegrees=10}
 {bone1Name=dyn_branch_r_2, bone2Name=dyn_base_r, swing1LimitDegrees=20, swing2LimitDegrees=20, twistLimitDegrees=10}
 {bone1Name=dyn_branch_r_3, bone2Name=dyn_base_r, swing1LimitDegrees=20, swing2LimitDegrees=20, twistLimitDegrees=10}
 {bone1Name=dyn_branch_r_4, bone2Name=dyn_base_r, swing1LimitDegrees=20, swing2LimitDegrees=20, twistLimitDegrees=10}
 {bone1Name=dyn_branch_r_5, bone2Name=dyn_base_r, swing1LimitDegrees=20, swing2LimitDegrees=20, twistLimitDegrees=10}
 {bone1Name=dyn_branch_r_6, bone2Name=dyn_base_r, swing1LimitDegrees=20, swing2LimitDegrees=20, twistLimitDegrees=10}
 {bone1Name=dyn_branch_r_7, bone2Name=dyn_base_r, swing1LimitDegrees=20, swing2LimitDegrees=20, twistLimitDegrees=10}
 {
  bone1Name=clavicle_r
  bone2Name=spine_05
  swing1LimitDegrees=45
  swing1Motion=Free
  swing2LimitDegrees=45
  swing2Motion=Free
  twistLimitDegrees=45
  twistMotion=Free
 }
 {
  bone1Name=upperarm_r
  bone2Name=clavicle_r
  swing1LimitDegrees=45
  swing1Motion=Free
  swing2LimitDegrees=45
  swing2Motion=Free
  twistLimitDegrees=45
  twistMotion=Free
 }
 {bone1Name=dyn_bangs, bone2Name=head, swing1LimitDegrees=3, swing2LimitDegrees=8, twistLimitDegrees=5}
 {bone1Name=dyn_tale_l_1, bone2Name=dyn_tale_l_0, swing1LimitDegrees=20, swing2LimitDegrees=5, twistLimitDegrees=2}
 {bone1Name=dyn_tale_l_0, bone2Name=head, swing1LimitDegrees=10, swing2LimitDegrees=2, twistLimitDegrees=2}
 {bone1Name=dyn_tale_r_0, bone2Name=head, swing1LimitDegrees=10, swing2LimitDegrees=2, twistLimitDegrees=2}
]
