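# Animation Blueprint for the female KeyTracker character's face accessory, following the attached parent's pose and curves and adding physics secondary motion.
# When customization loads, it caches custom mesh part index 1 as the body.
# It layers the body's copied pose onto the parent's pose from the upperarm_r branch.
# A RigidBody node then simulates relative to the head bone with damped component motion and overridden gravity.
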AnimBlueprint /BRCosmetics/Characters/Player/Female/Medium/Bodies/F_MED_KeyTracker/Meshes/Parts/F_MED_KeyTracker_FaceAcc_AnimBP
target_skeleton=&/F_MED_KeyTracker_FaceAcc_Skeleton
graphs [
 AnimGraph {nodes [
   n1 = Copy Pose From Mesh() {bCopyCurves=true, bUseAttachedParent=true}
   n2 = Copy Pose From Mesh(SourceMeshComponent=Get body() -> (body:Skeletal Mesh Component Budgeted Object Reference)) {bCopyCurves=true, bUseAttachedParent=true}
   n3 = RigidBody(ComponentPose=Local To Component(LocalPose=Layered blend per bone(BasePose=@n1, BlendPoses_0=@n2) {layerSetup [{branchFilters [{boneName=upperarm_r}]}]})) {
    bOverrideWorldGravity=true
    overridePhysicsAsset=&/F_MED_KeyTracker_FaceAcc_Physics
    simulationSpace=BaseBoneSpace
    alphaBoolBlend {blendOption=Linear}
    baseBoneRef {boneName=head}
    componentAppliedLinearAccClamp=<3000 3000 3000>
    componentLinearAccScale=<0.4 0.4 0.4>
    componentLinearVelScale=<0.4 0.4 0.4>
    overrideWorldGravity {z=-980}
    simSpaceSettings {maxAngularAcceleration=100, maxAngularVelocity=50, maxLinearAcceleration=1, maxLinearVelocity=10, velocityScaleZ=0, worldAlpha=1}
   }
   Output Pose(Result=Component To Local(ComponentPose=@n3)) {layerGroup=None}
  ]}
 EventGraph {nodes [
   n1 = Event OnCharacterCustomizationLoaded()
   > Set body(body:Skeletal Mesh Component Budgeted Object Reference="Get (a copy)"(Array:Array of Skeletal Mesh Component Budgeted Object References=@n1.CustomMeshParts, Dimension 1=1) -> (Output:Skeletal Mesh Component Budgeted Object Reference)) -> (Output_Get:Skeletal Mesh Component Budgeted Object Reference)
   Event BlueprintUpdateAnimation()
  ]}
]
variables [body {type=SkeletalMeshComponentBudgeted}]
